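; Gifted Traveler Blueprint that toggles World Partition data layers from calendar events and enables a boss spawn point.
; On authority BeginPlay, it awaits playlist and calendar sync, then runs ActivateEvent and EnableBossSpawn for active events.
; With none active, it unloads only the first entry's enabled and disabled layers, then activates default layers.
; GT_Event activates events by name, DeactivateEvent reverses layer states, and LogMessage writes to the shipping log.
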(event EventBeginPlay
  (bind _calendarevents (Variables|Default|GetCalendarEvents))
  (bind _output (Utilities|Array|Get(aref) _calendarevents))
  (bind (_eventname_2_90f02b394aaeb20f36d7a8849eeccb09 _enableddatalayers_9_41bb9a3e4648805b5af70aa728288c92 _disableddatalayers_10_8de59e3b474c88ed2e7988b185572444) (Utilities|Struct|BreakCalendarDataLayerData _output))
  (Networking|SwitchHasAuthority
    (:Authority
      (Variables|Default|SetFoundActiveEvent false)
      (Class|FortAsyncActionCurrentPlaylistReady|PlaylistDataReadyAsync
        (:OnReady
          (MCPContext|InitialCalendarSyncCompleted
            (:OnSyncComplete
              (bind _array_element (Utilities|Array|ForEachLoopwithBreak (Variables|Default|GetCalendarEvents))
                (:LoopBody
                  (bind _biseventactive (CalendarEvents|GetCalendarEventInformation (Utilities|Struct|BreakCalendarDataLayerData _array_element)))
                  (if _biseventactive
                    (Variables|Default|SetFoundActiveEvent true)
                    (CallFunction|EnableBossSpawn)
                    (CallFunction|LogMessage (Utilities|String|ToString(Text) (Utilities|Text|FormatText "NSLOCTEXT(\"[4C4B711C08C2F0A9F9FA3BA816AF9FAE]\", \"3580DE8544152B70CFBA1C87FDC9FF5E\", \"[BP_GiftedTraveler_CalendarDataLayer][EventGraph] Calendar event name {a} is ACTIVE, toggling data layers appropriately.\")" (Utilities|Struct|BreakCalendarDataLayerData _array_element))))
                    (CallFunction|ActivateEvent (Utilities|Struct|BreakCalendarDataLayerData _array_element))))
                (:Completed
                  (if (not (Variables|Default|GetFoundActiveEvent))
                    (CallFunction|LogMessage (Utilities|String|ToString(Text) (Utilities|Text|FormatText "NSLOCTEXT(\"[4C4B711C08C2F0A9F9FA3BA816AF9FAE]\", \"DA69F0A84142619257205B8192712848\", \"[BP_GiftedTraveler_CalendarDataLayer][EventGraph] No active events found. Disabling all the data layers found on the event.\")")))
                    (Utilities|FlowControl|Sequence
                      (:then_0
                        (for _array_element_2 _enableddatalayers_9_41bb9a3e4648805b5af70aa728288c92
                          (bind _returnvalue (WorldPartition|GetDataLayerManager))
                          (bind _returnvalue_2 (DataLayers|SetDataLayerRuntimeState _returnvalue _array_element_2 "Unloaded" false :!class /Script/Engine.DataLayerManager))
                          (CallFunction|LogMessage (Utilities|String|ToString(Text) (Utilities|Text|FormatText "NSLOCTEXT(\"[4C4B711C08C2F0A9F9FA3BA816AF9FAE]\", \"048E9A8742B0B04355FD3B81EBE57DD5\", \"[BP_GiftedTraveler_CalendarDataLayer][EventGraph] Unloading data layer {a}. Succeeded: {b}\")" (Utilities|GetDisplayName _array_element_2) _returnvalue_2))))
                        (for _array_element_3 _disableddatalayers_10_8de59e3b474c88ed2e7988b185572444
                          (bind _returnvalue_3 (WorldPartition|GetDataLayerManager))
                          (bind _returnvalue_4 (DataLayers|SetDataLayerRuntimeState _returnvalue_3 _array_element_3 "Unloaded" false :!class /Script/Engine.DataLayerManager))
                          (CallFunction|LogMessage (Utilities|String|ToString(Text) (Utilities|Text|FormatText "NSLOCTEXT(\"[4C4B711C08C2F0A9F9FA3BA816AF9FAE]\", \"A5F330AB4E51653C13DA8CAC26AB57A9\", \"[BP_GiftedTraveler_CalendarDataLayer][EventGraph] Unloading data layer {a}. Succeeded: {b}\")" (Utilities|GetDisplayName _array_element_3) _returnvalue_4)))))
                      (:then_1
                        (CallFunction|LogMessage (Utilities|String|ToString(Text) (Utilities|Text|FormatText "NSLOCTEXT(\"[4C4B711C08C2F0A9F9FA3BA816AF9FAE]\", \"9D1E52064DE2B88C98C9DCB864722A2D\", \"[BP_GiftedTraveler_CalendarDataLayer][EventGraph] No active events found. Enabling data layers that should be on by default.\")")))
                        (for _array_element_4 (Variables|Default|GetDefaultEnabledDataLayers)
                          (bind _returnvalue_5 (WorldPartition|GetDataLayerManager))
                          (bind _returnvalue_6 (DataLayers|SetDataLayerRuntimeState _returnvalue_5 _array_element_4 "Activated" false :!class /Script/Engine.DataLayerManager))
                          (CallFunction|LogMessage (Utilities|String|ToString(Text) (Utilities|Text|FormatText "NSLOCTEXT(\"[4C4B711C08C2F0A9F9FA3BA816AF9FAE]\", \"75810840438501FFE8247AAC58C9CA2E\", \"[BP_GiftedTraveler_CalendarDataLayer][EventGraph] Activating data layer {a}. Succeeded: {b}\")" (Utilities|GetDisplayName _array_element_4) _returnvalue_6))))))))))))))) ; The Client won't get past the\nCalendar checking. Also, the replicated\nsequence player runs on the server\nand replicates to the client.
)

(event Custom|EnableBossSpawn
  (Class|FortAthenaLivingWorldPointProviderInterface|EnablePointProvider_DEPRECATED (Variables|Default|GetBossSpawnPoint)))

(event Custom|GT_Event
  (params (EventName :type String))
  (CallFunction|ActivateEvent EventName))

(fn ActivateEvent
  (params (EventName :type String))
  (for _array_element (Variables|Default|GetCalendarEvents)
    (if (Utilities|String|EqualExactly(String) (Variables|GetEventName) (Utilities|Struct|BreakCalendarDataLayerData _array_element))
      (CallFunction|LogMessage (Utilities|String|ToString(Text) (Utilities|Text|FormatText "NSLOCTEXT(\"[4C4B711C08C2F0A9F9FA3BA816AF9FAE]\", \"5C8548064D8F5F44562C508D4EDF2CE1\", \"[BP_GiftedTraveler_CalendarDataLayer][ActivateEvent] Activating event {a}\")" (Variables|GetEventName))))
      (Utilities|FlowControl|Sequence
        (:then_0
          (for _array_element_2 (.EnabledDataLayers_9_41BB9A3E4648805B5AF70AA728288C92 (Utilities|Struct|BreakCalendarDataLayerData _array_element))
            (bind _returnvalue (WorldPartition|GetDataLayerManager))
            (bind _returnvalue_2 (DataLayers|SetDataLayerRuntimeState _returnvalue _array_element_2 "Activated" false :!class /Script/Engine.DataLayerManager))
            (CallFunction|LogMessage (Utilities|String|ToString(Text) (Utilities|Text|FormatText "NSLOCTEXT(\"[4C4B711C08C2F0A9F9FA3BA816AF9FAE]\", \"2289AC324BDC58E9AC9B7D8B83306870\", \"[BP_GiftedTraveler_CalendarDataLayer][ActivateEvent] Activating data layer {a}. Succeeded: {b}\")" (Utilities|GetDisplayName _array_element_2) _returnvalue_2)))))
        (:then_1
          (for _array_element_3 (.DisabledDataLayers_10_8DE59E3B474C88ED2E7988B185572444 (Utilities|Struct|BreakCalendarDataLayerData _array_element))
            (bind _returnvalue_3 (WorldPartition|GetDataLayerManager))
            (bind _returnvalue_4 (DataLayers|SetDataLayerRuntimeState _returnvalue_3 _array_element_3 "Unloaded" false :!class /Script/Engine.DataLayerManager))
            (CallFunction|LogMessage (Utilities|String|ToString(Text) (Utilities|Text|FormatText "NSLOCTEXT(\"[4C4B711C08C2F0A9F9FA3BA816AF9FAE]\", \"13A8418C4E16CFB423BDFA938DEB0510\", \"[BP_GiftedTraveler_CalendarDataLayer][ActivateEvent] Unloading data layer {a}. Succeeded: {b}\")" (Utilities|GetDisplayName _array_element_3) _returnvalue_4)))))))))

(fn DeactivateEvent
  (params (EventName :type String))
  (for _array_element (Variables|Default|GetCalendarEvents)
    (if (Utilities|String|EqualExactly(String) (Variables|GetEventName) (Utilities|Struct|BreakCalendarDataLayerData _array_element))
      (Utilities|FlowControl|Sequence
        (:then_0
          (for _array_element_2 (.EnabledDataLayers_9_41BB9A3E4648805B5AF70AA728288C92 (Utilities|Struct|BreakCalendarDataLayerData _array_element))
            (bind _returnvalue (WorldPartition|GetDataLayerManager))
            (DataLayers|SetDataLayerRuntimeState _returnvalue _array_element_2 "Unloaded" false :!class /Script/Engine.DataLayerManager)))
        (:then_1
          (for _array_element_3 (.DisabledDataLayers_10_8DE59E3B474C88ED2E7988B185572444 (Utilities|Struct|BreakCalendarDataLayerData _array_element))
            (bind _returnvalue_2 (WorldPartition|GetDataLayerManager))
            (DataLayers|SetDataLayerRuntimeState _returnvalue_2 _array_element_3 "Activated" false :!class /Script/Engine.DataLayerManager)))))))

(fn LogMessage
  (params (Message :type String))
  (Fortnite|FortShippingLog self Message))

(fn ConstructionScript
  (|CallParentFunction))
