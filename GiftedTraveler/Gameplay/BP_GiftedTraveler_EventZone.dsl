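; Event zone actor for GiftedTraveler that applies temporary effects to players and vehicles inside its EventZone volume.
; On server authority, overlap start applies PlayerEffect to the player pawn's ability system and overlap end removes it.
; Entering vehicles are woken, stopped and given replicated loose VehicleTags, which are removed on exit.
; The volume turns on query-only collision at BeginPlay, and SetEnabled switches that collision on or off.
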(event EventBeginPlay
  (Collision|SetCollisionEnabled (Variables|Default|GetEventZone) "QueryOnly"))

(event Collision|EventActorBeginOverlap
  (params (OtherActor :type Actor%20Object%20Reference)))

(event EventTick
  (params (DeltaSeconds :type Float%20%28single-precision%29)))

(event OnComponentBeginOverlap(EventZone)
  (params (OverlappedComponent :type Primitive%20Component%20Object%20Reference) (OtherActor :type Actor%20Object%20Reference) (OtherComp :type Primitive%20Component%20Object%20Reference) (OtherBodyIndex :type Integer) (bFromSweep :type Boolean) (SweepResult :type Hit%20Result%20Structure%20%28by%20ref%29))
  (Networking|SwitchHasAuthority
    (:Authority
      (CallFunction|ApplyZoneEffectstoPlayer OtherActor)
      (CallFunction|SetVehicleEnabled OtherActor))))

(event OnComponentEndOverlap(EventZone)
  (params (OverlappedComponent :type Primitive%20Component%20Object%20Reference) (OtherActor :type Actor%20Object%20Reference) (OtherComp :type Primitive%20Component%20Object%20Reference) (OtherBodyIndex :type Integer))
  (Networking|SwitchHasAuthority
    (:Authority
      (CallFunction|RemoveZoneEffectsfromPlayer OtherActor)
      (CallFunction|SetVehicleEnabled OtherActor true))))

(fn ApplyZoneEffectsToPlayer
  (params (Actor :type Actor%20Object%20Reference))
  (bind _asfort_player_pawn_athena (Utilities|Casting|CastToFortPlayerPawnAthena Actor))
  (GameplayEffects|ApplyGameplayEffectToTarget :self (Ability|GetFortAbilitySystemComponent _asfort_player_pawn_athena) :GameplayEffectClass (Variables|Default|GetPlayerEffect) :Target (Ability|GetFortAbilitySystemComponent _asfort_player_pawn_athena) :Level 0.0))

(fn RemoveZoneEffectsFromPlayer
  (params (Actor :type Actor%20Object%20Reference))
  (bind _asfort_player_pawn_athena (Utilities|Casting|CastToFortPlayerPawnAthena Actor))
  (GameplayEffects|RemoveActiveGameplayEffectbySourceEffect (Ability|GetFortAbilitySystemComponent _asfort_player_pawn_athena) (Variables|Default|GetPlayerEffect) 0 -1))

(fn SetEnabled
  (params (Enabled :type Boolean))
  (Collision|SetCollisionEnabled (Variables|Default|GetEventZone) (select Enabled "QueryOnly" "NoCollision")))

(fn SetVehicleEnabled
  (params (Actor :type Actor%20Object%20Reference) (Enabled :type Boolean))
  (bind _vehicletags (Variables|Default|GetVehicleTags))
  (bind _asfort_athena_vehicle (Utilities|Casting|CastToFortAthenaVehicle Actor))
  (if Enabled
    (Ability|GameplayEffect|RemoveLooseGameplayTags Actor _vehicletags true)
    (else
      (Vehicle|WakeUp _asfort_athena_vehicle)
      (Components|Movement|StopMovementImmediately (Pawn|GetMovementComponent _asfort_athena_vehicle))
      (Ability|GameplayEffect|AddLooseGameplayTags Actor _vehicletags true))))

(fn ConstructionScript)
