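; BP_GiftedTraveler_LeadActor is the event's skeletal-mesh character, reacting to VividRazor Baron events beside a portal.
; After match-ready and calendar sync, the server stores routed Baron events and shows delayed speech bubbles.
; One event's OnRep triggers a one-time reaction montage, using the final version when the Phase 5 data layer is active.
; Montage notifies drive portal impact, shutdown, disappearance with event-zone swap and broadcasts, plus replicated music-encounter cues.
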(event EventBeginPlay
  (bind _combatreactiondelay (Variables|Default|GetCombatReactionDelay))
  (bind _returnvalue (ScalableFloat|GetValueatLevel _combatreactiondelay))
  (Networking|SwitchHasAuthority
    (:Authority
      (Class|BPGiftedTravelerPortal|SetPortalIsActive (Variables|Default|GetPortalActor) true)
      (Class|BPGiftedTravelerEventZone|SetEnabled (Variables|Default|GetEventZone_Portal) false)
      (Class|FortAsyncActionAuthorityMatchReady|MatchReadyAsync
        (:OnReady
          (MCPContext|InitialCalendarSyncCompleted
            (:OnSyncComplete
              (bind (_asynctaskproxy _payload) (GameplayEventRouter|StartListeningToStatefulEvent (GameplayEventRouter|GetGlobalEventRouter) "/VividRazor_Baron/Gameplay/EventRouter/VividRazor_Baron_Event.VividRazor_Baron_Event")
                (:OnEventReceived
                  (Variables|Default|SetLatestBaronEvent (Utilities|Struct|BreakVividRazorBaronEvent _payload))
                  (if (> _returnvalue 0)
                    (Utilities|FlowControl|Delay _returnvalue)
                    (else
                      _))
                  (CallFunction|ShowSpeechBubble))))))))
    (:Remote
      (Animation|Montage|PlayMontage (Variables|SkeletalMeshActor|GetSkeletalMeshComponent) (Variables|Default|GetLoopMontage)))))

(event Custom|PlayReactionMontage
  (bind _notifyname (Animation|Montage|PlayMontage (Variables|SkeletalMeshActor|GetSkeletalMeshComponent) (select (CallFunction|CheckIsLastDay) (Variables|Default|GetFinalReactionMontage) (Variables|Default|GetDefaultReactionMontage)))
    (:OnNotifyBegin
      (bind _output_get (Variables|Default|SetMontageNotify _notifyname))
      (switch name _output_get
        (:Disappear
          (Rendering|SetActorHiddenInGame true)
          (Networking|SwitchHasAuthority
            (:Authority
              (bind (_arrayindex _arrayelement) (Utilities|Array|ReverseforEachLoop (Actor|GetAttachedActors false true))
                (:LoopBody
                  (Actor|DestroyActor _arrayelement))
                (:Completed
                  (CustomVerb|BroadcastCustomVerbMessage (CustomVerb|MakeCustomVerbMessage self "(TagName=\"CustomVerb.GiftedTraveler.MMDisappeared\")"))
                  (Class|BPGiftedTravelerEventZone|SetEnabled (Variables|Default|GetEventZone) false)
                  (Class|BPGiftedTravelerEventZone|SetEnabled (Variables|Default|GetEventZone_Portal) true)
                  (CallFunction|BroadcastDisappearance))))))
        (:PortalImpact
          (Class|BPGiftedTravelerPortal|OnHItbyMM (Variables|Default|GetPortalActor)))
        (:PortalOff
          (Networking|SwitchHasAuthority
            (:Authority
              (Class|BPGiftedTravelerPortal|SetPortalIsActive (Variables|Default|GetPortalActor) false))))))))

(event Custom|OnDefeatReact
  (bind _combatreactiondelay (Variables|Default|GetCombatReactionDelay))
  (bind _returnvalue (ScalableFloat|GetValueatLevel _combatreactiondelay))
  (if (> _returnvalue 0)
    (Utilities|FlowControl|Delay _returnvalue)
    (else
      _))
  (CallFunction|PlayReactionMontage)
  (GameplayEventRouter|BroadcastStatefulEvent (GameplayEventRouter|GetGlobalEventRouter) (Utilities|Struct|MakeVividRazorBaronEvent "NewEnumerator4") 0))

(event Custom|BroadcastDisappearance
  (GameplayEventRouter|BroadcastStatefulEvent (GameplayEventRouter|GetGlobalEventRouter) (Utilities|Struct|MakeVividRazorBaronEvent "NewEnumerator5") 0))

(fn CheckIsLastDay
  (bind _returnvalue (WorldPartition|GetDataLayerManager))
  (return (Utilities|Enum|Equal(Enum) (DataLayers|GetDataLayerInstanceRuntimeState _returnvalue (DataLayers|GetDataLayerInstancefromAsset _returnvalue "/GiftedTraveler/DataLayers/DL_GiftedTraveler_Phase5.DL_GiftedTraveler_Phase5")) "Activated")))

(fn OnRep_GuardEliminated)

(fn OnRep_GuardThreatened)

(fn OnRep_LatestBaronEvent
  (switch Utilities|FlowControl|Switch|SwitchonVividRazor_Baron_Event_Enum (Variables|Default|GetLatestBaronEvent)
    (:NewEnumerator1
      (Utilities|FlowControl|DoOnce false
        (:Completed
          (CallFunction|OnDefeatReact))))))

(fn OnRep_MontageNotify
  (Networking|CanExecuteCosmeticEvents
    (:True
      (MusicEncounter|ExecuteMusicEncounterGameplayEventByName (GameplayEventRouter|GetGlobalEventRouter) (Variables|Default|GetMontageNotify) self))))

(fn ShowSpeechBubble
  (Class|BPGiftedTravelerSpeechBubble|ShowSpeechBubble (Variables|Default|GetSpeechBubbleActor) (Variables|Default|GetLatestBaronEvent)))

(fn ConstructionScript)
