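; Music spawner Blueprint for the Gifted Traveler that drives music encounter events from the VividRazor Baron encounter state.
; BeginPlay listens to the global stateful Baron event and stores its payload in BaronEvent.
; The BaronEvent OnRep executes the music encounter event named after the enum value, only where cosmetic events can run.
; InitializeMusic separately fires the "Start" music event through the same global event router.
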(event EventBeginPlay
  (bind (_asynctaskproxy _payload) (GameplayEventRouter|StartListeningToStatefulEvent (GameplayEventRouter|GetGlobalEventRouter) "/VividRazor_Baron/Gameplay/EventRouter/VividRazor_Baron_Event.VividRazor_Baron_Event")
    (:OnEventReceived
      (Variables|Default|SetBaronEvent (Utilities|Struct|BreakVividRazorBaronEvent _payload)))))

(event Custom|InitializeMusic
  (MusicEncounter|ExecuteMusicEncounterGameplayEventByName (GameplayEventRouter|GetGlobalEventRouter) "Start" self))

(fn OnRep_BaronEvent
  (Networking|CanExecuteCosmeticEvents
    (:True
      (MusicEncounter|ExecuteMusicEncounterGameplayEventByName (GameplayEventRouter|GetGlobalEventRouter) (Utilities|Name|EnumtoName (Variables|Default|GetBaronEvent)) self))))

(fn ConstructionScript)
