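; Gifted Traveler portal actor with gate VFX, looping audio and an impulse sphere that repels overlapping actors.
; While PortalIsActive, the server knocks back non-NPC players via a gameplay effect and tosses pickups.
; ChangePortalState sets sphere collision, then either plays a timed turn-off with fade and explosion sound or toggles visuals instantly; clients start hidden.
; OnHItByMM triggers the Niagara hit state and final explosion sound.
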(event EventBeginPlay
  (|CallParentFunction)
  (Networking|SwitchHasAuthority
    (:Authority
      _)
    (:Remote
      (CallFunction|TogglePortalStateExplicit false)))
  (Networking|CanExecuteCosmeticEvents
    (:True
      (Audio|Components|Audio|Play (Variables|Default|GetLoopingPortalAudio) 0.0))))

(event Collision|EventActorBeginOverlap
  (params (OtherActor :type Actor%20Object%20Reference)))

(event EventTick
  (params (DeltaSeconds :type Float%20%28single-precision%29)))

(event Custom|OnHItByMM
  (Niagara|SetNiagaraVariable(Bool) (Variables|Default|GetNE_SystemShare) "IsHit" true)
  (CallFunction|PlayFinalPortalExplosionSound))

(event OnComponentBeginOverlap(ImpulseSphere)
  (params (OverlappedComponent :type Primitive%20Component%20Object%20Reference) (OtherActor :type Actor%20Object%20Reference) (OtherComp :type Primitive%20Component%20Object%20Reference) (OtherBodyIndex :type Integer) (bFromSweep :type Boolean) (SweepResult :type Hit%20Result%20Structure%20%28by%20ref%29))
  (Networking|SwitchHasAuthority
    (:Authority
      (if (Variables|Default|GetPortalIsActive)
        (CallFunction|PushActorAway OtherActor)))))

(event Custom|Event_TurnPortalOnOff
  (params (OnOff :type Boolean))
  (bind _ne_systemshare (Variables|Default|GetNE_SystemShare))
  (Components|Activation|Activate (Variables|Default|GetNS_GiftedTraveler_PortalTurnOff) true)
  (Effects|Components|ParticleSystem|SetBoolParameter (Variables|Default|GetNE_SystemShare) "FadeOut" (not OnOff))
  (Audio|Components|Audio|FadeOut (Variables|Default|GetLoopingPortalAudio) (Variables|Default|GetFadeOutDuration) 0.0 "Linear" :!class /Script/Engine.AudioComponent)
  (Audio|SpawnSoundAttached (Variables|Default|GetExplosionSound) (Variables|Default|GetLoopingPortalAudio))
  (Utilities|FlowControl|Delay (Variables|Default|GetFadeOutDuration))
  (Rendering|SetVisibility (Variables|Default|GetVFX_Gate) OnOff false)
  (if OnOff
    (Components|Activation|Activate _ne_systemshare true)
    (else
      (Components|Activation|Deactivate _ne_systemshare))))

(fn ChangePortalState
  (params (Active :type Boolean))
  (bind _active (Variables|GetActive))
  (Collision|SetCollisionEnabled (Variables|Default|GetImpulseSphere) (select Active "QueryOnly" "NoCollision"))
  (if (and (Rendering|IsVisible (Variables|Default|GetVFX_Gate)) (not _active))
    (CallFunction|EventTurnPortalonOff _active)
    (else
      (CallFunction|TogglePortalStateExplicit _active))))

(fn PlayFinalPortalExplosionSound
  (Audio|SpawnSoundAttached (Variables|Default|GetLastDayExplosionSound) (Variables|Default|GetLoopingPortalAudio)))

(fn PushActorAway
  (params (Actor :type Actor%20Object%20Reference))
  (bind _asfort_player_pawn_athena (Utilities|Casting|CastToFortPlayerPawnAthena Actor)
    (:then
      (if (not (GameplayTags|HasMatchingGameplayTag _asfort_player_pawn_athena "(TagName=\"Pawn.Athena.NPC\")"))
        (GameplayEffects|FortApplyGameplayEffectToSelf :self (Ability|GetFortAbilitySystemComponent _asfort_player_pawn_athena) :GameplayEffectClass "/GiftedTraveler/Assets/Props/Orb/GameplayCues/GE_MM_GrantKnockback.GE_MM_GrantKnockback_C" :OptionalSourceObject self :OptionalEffectCauser self :Level 0.0)))
    (:CastFailed
      (bind _asfort_pickup (Utilities|Casting|CastToFortPickup Actor))
      (bind (_returnvalue_roll _returnvalue_pitch _returnvalue_yaw) (Math|Rotator|BreakRotator (Math|Rotator|FindLookatRotation (Transformation|GetActorLocation) (Transformation|GetActorLocation _asfort_pickup))))
      (Pickup|TossPickupinDirection _asfort_pickup (* (Math|Vector|MakeVector _returnvalue_roll _returnvalue_pitch (Math|Float|Absolute(Float) _returnvalue_yaw)) 5.0) 0 false true))))

(fn SetPortalIsActive
  (params (PortalActive :type Boolean))
  (Variables|Default|SetPortalIsActive PortalActive))

(fn SetPortalPreviousState
  (params (PreviousState :type Boolean))
  (Variables|Default|SetPreviousState PreviousState))

(fn SetVehicleEnabled
  (params (Actor :type Actor%20Object%20Reference) (Enabled :type Boolean))
  (bind _vehicletags (Variables|Default|GetVehicleTags))
  (bind _asfort_athena_vehicle (Utilities|Casting|CastToFortAthenaVehicle Actor))
  (if Enabled
    (Ability|GameplayEffect|RemoveLooseGameplayTags Actor _vehicletags true)
    (else
      (Vehicle|WakeUp _asfort_athena_vehicle)
      (Components|Movement|StopMovementImmediately (Pawn|GetMovementComponent _asfort_athena_vehicle))
      (Ability|GameplayEffect|AddLooseGameplayTags Actor _vehicletags true))))

(fn TogglePortalStateExplicit
  (params (State :type Boolean))
  (bind _ne_systemshare (Variables|Default|GetNE_SystemShare))
  (Rendering|SetVisibility (Variables|Default|GetVFX_Gate) State false)
  (if State
    (Components|Activation|Activate _ne_systemshare true)
    (else
      (Components|Activation|Deactivate _ne_systemshare))))

(fn ConstructionScript
  (|CallParentFunction))
