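; Gifted Traveler speech bubble Blueprint that starts speech matching the latest Baron event.
; ShowSpeechBubble picks the final or default event-to-tag map, depending on whether the Phase 5 data layer is activated.
; It passes the mapped gameplay tag as a query to the FortPawn speech bubble component, using a checklist map to skip events already marked.
; BeginPlay, overlap, Tick and the construction script are empty.
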(event EventBeginPlay)

(event Collision|EventActorBeginOverlap
  (params (OtherActor :type Actor%20Object%20Reference)))

(event EventTick
  (params (DeltaSeconds :type Float%20%28single-precision%29)))

(fn CheckIsLastDay
  (bind _returnvalue (WorldPartition|GetDataLayerManager))
  (return (Utilities|Enum|Equal(Enum) (DataLayers|GetDataLayerInstanceRuntimeState _returnvalue (DataLayers|GetDataLayerInstancefromAsset _returnvalue "/GiftedTraveler/DataLayers/DL_GiftedTraveler_Phase5.DL_GiftedTraveler_Phase5")) "Activated")))

(fn ShowSpeechBubble
  (params (LatestBaronEvent :type VividRazor_Baron_Event_Enum%20Enum :default "NewEnumerator3"))
  (bind (_value _returnvalue) (Utilities|Map|Find (Variables|Default|GetSpeechMapChecklist) (Variables|GetLatestBaronEvent)))
  (bind _latestbaronevent (Variables|GetLatestBaronEvent))
  (bind _output_get (LocalVariables|SetActiveSpeechMap (select (Default|CheckIsLastDay) (Variables|Default|GetFinalSpeechMap) (Variables|Default|GetDefaultSpeechMap))))
  (if (and (.ReturnValue (Utilities|Map|Find _output_get _latestbaronevent)) (or (not _value) (not _returnvalue)))
    (SpeechBubble|StartSpeech (Variables|Default|GetFortPawnComponent_SpeechBubble) (GameplayTags|MakeGameplayTagQueryMatchAnyTags (GameplayTags|MakeGameplayTagContainerfromArray (Utilities|Array|MakeArray (Utilities|Map|Find _output_get _latestbaronevent)))))
    (Utilities|Map|Add (Variables|Default|GetSpeechMapChecklist) (Variables|GetLatestBaronEvent) "true")))

(fn ConstructionScript)
