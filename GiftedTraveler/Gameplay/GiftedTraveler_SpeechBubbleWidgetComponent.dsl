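; Speech bubble widget component for the Gifted Traveler that adjusts the bubble's vertical offset according to local player distance.
; UpdatedWidgetOffset measures the distance from the owning actor to the first local player's pawn, if valid.
; It normalizes and clamps that distance, then lerps the widget's render translation so the bubble rises as the player moves away.
; BeginPlay and Tick do nothing, and nothing here calls the offset event.
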(event EventBeginPlay
  (|CallParentFunction) ; This node is disabled and will not be called.\nDrag off pins to build functionality.
)

(event EventTick
  (params (DeltaSeconds :type Float%20%28single-precision%29)))

(event EventUpdatedWidgetOffset
  (bind _returnvalue (Fortnite|GetFirstLocalFortPlayerController true))
  (bind _returnvalue_2 (Pawn|GetControlledPawn _returnvalue))
  ;[1 1 1] UpdateWidgetOffset
  (Utilities|IsValid _returnvalue_2
    (:Is%20Valid
      (Variables|/Script/UMG.Widget|SetRenderTransform_Translation_X (Utilities|Struct|MakeWidgetTransform (Math|Vector2D|MakeVector2D 0.0 (Math|Float|Lerp -1e+02 -2e+02 (Math|Float|Clamp(Float) (Math|Float|NormalizetoRange (Math|Vector|Distance(Vector) (Transformation|GetActorLocation (Components|GetOwner)) (Transformation|GetActorLocation _returnvalue_2)) 0.0 5e+03))))) (UserInterface|GetUserWidgetObject)))))
